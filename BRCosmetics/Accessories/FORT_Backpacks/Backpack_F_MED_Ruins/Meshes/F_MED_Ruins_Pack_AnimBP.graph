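# Animation Blueprint for the F_MED_Ruins winged backpack that poses its wings to match player movement.
# EventGraph derives Dive (sprint, targeting, crouch, skydive dive), Fly (skydive floating) and Jump (outside skydiving) alphas from state flags.
# These drive interpolated additive bone rotations, with Anim Dynamics on dyn_wing bones between the Fly and Jump layers, weighted lower while skydiving.
# Index-finger jump rotations have zero alpha, leaving them inactive.
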AnimBlueprint /BRCosmetics/Accessories/FORT_Backpacks/Backpack_F_MED_Ruins/Meshes/F_MED_Ruins_Pack_AnimBP
target_skeleton=&/F_MED_Ruins_Pack_Skeleton
graphs [
 AnimGraph {nodes [
   n1 = "Transform (Modify) Bone - Bone: backpack_root"(ComponentPose=Local To Component(LocalPose=Input Pose()), Alpha=1, Translation="0.000,0,5", Rotation="0,0,0", Scale="1,1,1") {translationMode=BMM_Additive, alphaBoolBlend {blendOption=Linear}, boneToModify {boneName=backpack_root}, translation {z=5}}
   n2 = Get DivePose_alpha() -> (DivePose_alpha:double)
   n3 = "Transform (Modify) Bone - Bone: wing_shoulder_r"!(ComponentPose=@n1, Alpha=@n2, Rotation="-0.000,-45.000,0.000") -> (Pose) {
    rotationMode=BMM_Additive
    alphaBoolBlend {blendOption=Linear}
    alphaScaleBiasClamp {bInterpResult=true, interpSpeedDecreasing=2, interpSpeedIncreasing=15}
    boneToModify {boneName=wing_shoulder_r}
    rotation {yaw=-45}
   }
   n4 = "Transform (Modify) Bone - Bone: wing_shoulder_l"!(ComponentPose=@n3, Alpha=@n2, Rotation="-0.000,45,0.000") -> (Pose) {
    rotationMode=BMM_Additive
    alphaBoolBlend {blendOption=Linear}
    alphaScaleBiasClamp {bInterpResult=true, interpSpeedDecreasing=2, interpSpeedIncreasing=15}
    boneToModify {boneName=wing_shoulder_l}
    rotation {yaw=45}
   }
   n5 = "Transform (Modify) Bone - Bone: wing_wrist_r"!(ComponentPose=@n4, Alpha=@n2, Rotation="0.000,0.000,-20") -> (Pose) {
    rotationMode=BMM_Additive
    alphaBoolBlend {blendOption=Linear}
    alphaScaleBiasClamp {bInterpResult=true, interpSpeedDecreasing=2, interpSpeedIncreasing=15}
    boneToModify {boneName=wing_wrist_r}
    rotation {roll=-20}
   }
   n6 = "Transform (Modify) Bone - Bone: wing_wrist_l"!(ComponentPose=@n5, Alpha=@n2, Rotation="-0.000,0,-20") -> (Pose) {
    rotationMode=BMM_Additive
    alphaBoolBlend {blendOption=Linear}
    alphaScaleBiasClamp {bInterpResult=true, interpSpeedDecreasing=2, interpSpeedIncreasing=15}
    boneToModify {boneName=wing_wrist_l}
    rotation {roll=-20}
   }
   n7 = Get FlyPose_alpha() -> (FlyPose_alpha:double)
   n8 = "Transform (Modify) Bone - Bone: wing_shoulder_r"(ComponentPose=@n6, Alpha=@n7, Rotation="-18,15,-10") {
    rotationMode=BMM_Additive
    alphaBoolBlend {blendOption=Linear}
    alphaScaleBiasClamp {bClampResult=true, bInterpResult=true, clampMax=0.8, interpSpeedIncreasing=15}
    boneToModify {boneName=wing_shoulder_r}
    rotation=<rot -18 15 -10>
   }
   n9 = "Transform (Modify) Bone - Bone: wing_elbow_r"!(ComponentPose=@n8, Alpha=@n7, Rotation="-5,-3,-4") -> (Pose) {
    rotationMode=BMM_Additive
    alphaBoolBlend {blendOption=Linear}
    alphaScaleBiasClamp {bClampResult=true, bInterpResult=true, clampMax=0.25, interpSpeedIncreasing=15}
    boneToModify {boneName=wing_elbow_r}
    rotation=<rot -5 -3 -4>
   }
   n10 = "Transform (Modify) Bone - Bone: wing_wrist_r"(ComponentPose=@n9, Alpha=@n7, Rotation="52,22,45") {
    rotationMode=BMM_Additive
    alphaBoolBlend {blendOption=Linear}
    alphaScaleBiasClamp {bClampResult=true, bInterpResult=true, clampMax=0.4, interpSpeedIncreasing=15}
    boneToModify {boneName=wing_wrist_r}
    rotation=<rot 52 22 45>
   }
   n11 = "Transform (Modify) Bone - Bone: wing_shoulder_l"(ComponentPose=@n10, Alpha=@n7, Rotation="18,-15,-10") {
    rotationMode=BMM_Additive
    alphaBoolBlend {blendOption=Linear}
    alphaScaleBiasClamp {bClampResult=true, bInterpResult=true, clampMax=0.8, interpSpeedIncreasing=15}
    boneToModify {boneName=wing_shoulder_l}
    rotation=<rot 18 -15 -10>
   }
   n12 = "Transform (Modify) Bone - Bone: wing_elbow_l"!(ComponentPose=@n11, Alpha=@n7, Rotation="5,3,-4") -> (Pose) {
    rotationMode=BMM_Additive
    alphaBoolBlend {blendOption=Linear}
    alphaScaleBiasClamp {bClampResult=true, bInterpResult=true, clampMax=0.25, interpSpeedIncreasing=15}
    boneToModify {boneName=wing_elbow_l}
    rotation=<rot 5 3 -4>
   }
   n13 = "Transform (Modify) Bone - Bone: wing_wrist_l"(ComponentPose=@n12, Alpha=@n7, Rotation="-52,-22,45") {
    rotationMode=BMM_Additive
    alphaBoolBlend {blendOption=Linear}
    alphaScaleBiasClamp {bClampResult=true, bInterpResult=true, clampMax=0.4, interpSpeedIncreasing=15}
    boneToModify {boneName=wing_wrist_l}
    rotation=<rot -52 -22 45>
   }
   n14 = SelectFloat(A=0.5, B=0.8, bPickA=Get bIsSkyDiving())
   n15 = Anim Dynamics - Bone: dyn_wing_elbow_l(ComponentPose=@n13, Alpha=@n14, ExternalForce="0,0,0") {
    alpha=0.9
    angularDampingOverride=0.98
    angularSpringConstant=500
    bGravityOverrideInSimSpace=true
    bOverrideAngularDamping=true
    bOverrideLinearDamping=true
    gravityScale=0.125
    linearDampingOverride=0.98
    numSolverIterationsPostUpdate=4
    numSolverIterationsPreUpdate=16
    alphaBoolBlend {blendOption=Linear}
    alphaScaleBiasClamp {bInterpResult=true}
    boundBone {boneName=dyn_wing_elbow_l}
    componentLinearAccScale=<0.125 0.125 0.125>
    componentLinearVelScale=<0.125 0.125 0.125>
    physicsBodyDefinitions [{
      sphereCollisionRadius=0
      boundBone {boneName=dyn_wing_elbow_l}
      boxExtents=<18 18 7.5>
      constraintSetup {angularTargetAxis=AxisZ, twistAxis=AxisZ, angularLimitsMax=<0 10 0>, angularLimitsMin=<0 -20 0>, angularTarget=<0 0 1>}
      localJointOffset=<0 0 1>
     }]
   }
   n16 = Anim Dynamics - Bone: dyn_wing_elbow_r(ComponentPose=@n15, Alpha=@n14, ExternalForce="0,0,0") {
    alpha=0.9
    angularDampingOverride=0.98
    angularSpringConstant=500
    bGravityOverrideInSimSpace=true
    bOverrideAngularDamping=true
    bOverrideLinearDamping=true
    gravityScale=0.125
    linearDampingOverride=0.98
    numSolverIterationsPostUpdate=4
    numSolverIterationsPreUpdate=16
    alphaBoolBlend {blendOption=Linear}
    alphaScaleBiasClamp {bInterpResult=true}
    boundBone {boneName=dyn_wing_elbow_r}
    componentLinearAccScale=<0.125 0.125 0.125>
    componentLinearVelScale=<0.125 0.125 0.125>
    physicsBodyDefinitions [{
      sphereCollisionRadius=0
      boundBone {boneName=dyn_wing_elbow_r}
      boxExtents=<18 18 7.5>
      constraintSetup {angularTargetAxis=AxisZ, twistAxis=AxisZ, angularLimitsMax=<0 10 0>, angularLimitsMin=<0 -20 0>, angularTarget=<0 0 1>}
      localJointOffset=<0 0 -1>
     }]
   }
   n17 = Anim Dynamics - Bone: dyn_wing_wrist_l(ComponentPose=@n16, Alpha=@n14, ExternalForce="0,0,0") {
    alpha=0.9
    angularDampingOverride=0.98
    angularSpringConstant=500
    bAngularSpring=true
    bGravityOverrideInSimSpace=true
    bOverrideAngularDamping=true
    gravityScale=0.125
    linearDampingOverride=0.98
    numSolverIterationsPostUpdate=4
    numSolverIterationsPreUpdate=16
    alphaBoolBlend {blendOption=Linear}
    boundBone {boneName=dyn_wing_wrist_l}
    componentLinearAccScale=<0.08 0.08 0.08>
    componentLinearVelScale=<0.5 0.5 0.5>
    physicsBodyDefinitions [{
      sphereCollisionRadius=0
      boundBone {boneName=dyn_wing_wrist_l}
      boxExtents=<15 15 7.5>
      constraintSetup {angularTargetAxis=AxisZ, twistAxis=AxisZ, angularLimitsMax=<0 25 0>, angularLimitsMin=<0 -30 0>, angularTarget=<0 0 1>}
      localJointOffset=<0 0 1>
     }]
   }
   n18 = Anim Dynamics - Bone: dyn_wing_wrist_r(ComponentPose=@n17, Alpha=@n14, ExternalForce="0,0,0") {
    alpha=0.9
    angularDampingOverride=0.98
    angularSpringConstant=500
    bAngularSpring=true
    bGravityOverrideInSimSpace=true
    bOverrideAngularDamping=true
    gravityScale=0.125
    linearDampingOverride=0.98
    numSolverIterationsPostUpdate=4
    numSolverIterationsPreUpdate=16
    alphaBoolBlend {blendOption=Linear}
    boundBone {boneName=dyn_wing_wrist_r}
    componentLinearAccScale=<0.08 0.08 0.08>
    componentLinearVelScale=<0.5 0.5 0.5>
    physicsBodyDefinitions [{
      sphereCollisionRadius=0
      boundBone {boneName=dyn_wing_wrist_r}
      boxExtents=<15 15 7.5>
      constraintSetup {angularTargetAxis=AxisZ, twistAxis=AxisZ, angularLimitsMax=<0 25 0>, angularLimitsMin=<0 -30 0>, angularTarget=<0 0 1>}
      localJointOffset=<0 0 -1>
     }]
   }
   n19 = Anim Dynamics - Bone: dyn_wing_pinky_l_01(ComponentPose=@n18, Alpha=@n14, ExternalForce="0,0,0") {
    alpha=0.9
    angularDampingOverride=0.98
    angularSpringConstant=500
    bAngularSpring=true
    bGravityOverrideInSimSpace=true
    bOverrideAngularDamping=true
    gravityScale=0.125
    linearDampingOverride=0.98
    numSolverIterationsPostUpdate=4
    numSolverIterationsPreUpdate=16
    alphaBoolBlend {blendOption=Linear}
    boundBone {boneName=dyn_wing_pinky_l_01}
    componentLinearAccScale=<0.08 0.08 0.08>
    componentLinearVelScale=<0.5 0.5 0.5>
    physicsBodyDefinitions [{
      sphereCollisionRadius=0
      boundBone {boneName=dyn_wing_pinky_l_01}
      boxExtents=<5 5 2.5>
      constraintSetup {angularTargetAxis=AxisZ, twistAxis=AxisZ, angularLimitsMax=<0 7.5 0>, angularLimitsMin=<0 -5 0>, angularTarget=<0 0 1>}
      localJointOffset=<0 0 1>
     }]
   }
   n20 = Anim Dynamics - Bone: dyn_wing_pinky_r_01(ComponentPose=@n19, Alpha=@n14, ExternalForce="0,0,0") {
    alpha=0.9
    angularDampingOverride=0.98
    angularSpringConstant=500
    bAngularSpring=true
    bGravityOverrideInSimSpace=true
    bOverrideAngularDamping=true
    gravityScale=0.125
    linearDampingOverride=0.98
    numSolverIterationsPostUpdate=4
    numSolverIterationsPreUpdate=16
    alphaBoolBlend {blendOption=Linear}
    boundBone {boneName=dyn_wing_pinky_r_01}
    componentLinearAccScale=<0.08 0.08 0.08>
    componentLinearVelScale=<0.5 0.5 0.5>
    physicsBodyDefinitions [{
      sphereCollisionRadius=0
      boundBone {boneName=dyn_wing_pinky_r_01}
      boxExtents=<5 5 2.5>
      constraintSetup {angularTargetAxis=AxisZ, twistAxis=AxisZ, angularLimitsMax=<0 7.5 0>, angularLimitsMin=<0 -5 0>, angularTarget=<0 0 1>}
      localJointOffset=<0 0 -1>
     }]
   }
   n21 = Anim Dynamics - Bone: dyn_wing_index_l_01(ComponentPose=@n20, Alpha=@n14, ExternalForce="0,0,0") {
    alpha=0.9
    angularDampingOverride=0.98
    angularSpringConstant=500
    bAngularSpring=true
    bGravityOverrideInSimSpace=true
    bOverrideAngularDamping=true
    gravityScale=0.125
    linearDampingOverride=0.98
    numSolverIterationsPostUpdate=4
    numSolverIterationsPreUpdate=16
    alphaBoolBlend {blendOption=Linear}
    boundBone {boneName=dyn_wing_index_l_01}
    componentLinearAccScale=<0.08 0.08 0.08>
    componentLinearVelScale=<0.5 0.5 0.5>
    physicsBodyDefinitions [{
      sphereCollisionRadius=0
      boundBone {boneName=dyn_wing_index_l_01}
      boxExtents=<5 5 2.5>
      constraintSetup {angularTargetAxis=AxisZ, twistAxis=AxisZ, angularLimitsMax=<0 7.5 0>, angularLimitsMin=<0 -5 0>, angularTarget=<0 0 1>}
      localJointOffset=<0 0 1>
     }]
   }
   n22 = Anim Dynamics - Bone: dyn_wing_index_r_01(ComponentPose=@n21, Alpha=@n14, ExternalForce="0,0,0") {
    alpha=0.9
    angularDampingOverride=0.98
    angularSpringConstant=500
    bAngularSpring=true
    bGravityOverrideInSimSpace=true
    bOverrideAngularDamping=true
    gravityScale=0.125
    linearDampingOverride=0.98
    numSolverIterationsPostUpdate=4
    numSolverIterationsPreUpdate=16
    alphaBoolBlend {blendOption=Linear}
    boundBone {boneName=dyn_wing_index_r_01}
    componentLinearAccScale=<0.08 0.08 0.08>
    componentLinearVelScale=<0.5 0.5 0.5>
    physicsBodyDefinitions [{
      sphereCollisionRadius=0
      boundBone {boneName=dyn_wing_index_r_01}
      boxExtents=<5 5 2.5>
      constraintSetup {angularTargetAxis=AxisZ, twistAxis=AxisZ, angularLimitsMax=<0 7.5 0>, angularLimitsMin=<0 -5 0>, angularTarget=<0 0 1>}
      localJointOffset=<0 0 -1>
     }]
   }
   n23 = Get JumpPose_alpha() -> (JumpPose_alpha:double)
   n24 = "Transform (Modify) Bone - Bone: wing_shoulder_l"(ComponentPose=@n22, Alpha=@n23, Rotation="3.546,-29.936,-4.646") {
    rotationMode=BMM_Additive
    alphaBoolBlend {blendOption=Linear}
    alphaScaleBiasClamp {bClampResult=true, bInterpResult=true, clampMax=0.65, interpSpeedDecreasing=2, interpSpeedIncreasing=15}
    boneToModify {boneName=wing_shoulder_l}
    rotation=<rot 3.546 -29.936 -4.646>
   }
   n25 = "Transform (Modify) Bone - Bone: wing_wrist_l"(ComponentPose=@n24, Alpha=@n23, Rotation="-20.023,-4.529,15.773") {
    rotationMode=BMM_Additive
    alphaBoolBlend {blendOption=Linear}
    alphaScaleBiasClamp {bClampResult=true, bInterpResult=true, clampMax=0.65, interpSpeedDecreasing=2, interpSpeedIncreasing=15}
    boneToModify {boneName=wing_wrist_l}
    rotation=<rot -20.023 -4.529 15.773>
   }
   n26 = "Transform (Modify) Bone - Bone: wing_pinky_l_01"!(ComponentPose=@n25, Alpha=@n23, Rotation="-14.884,-4.138,12.483") -> (Pose) {
    rotationMode=BMM_Additive
    alphaBoolBlend {blendOption=Linear}
    alphaScaleBiasClamp {bClampResult=true, bInterpResult=true, clampMax=0.5}
    boneToModify {boneName=wing_pinky_l_01}
    rotation=<rot -14.884 -4.138 12.483>
   }
   n27 = "Transform (Modify) Bone - Bone: wing_ring_l_01"!(ComponentPose=@n26, Alpha=@n23, Rotation="-13.276,-5.550,4.957") -> (Pose) {
    rotationMode=BMM_Additive
    alphaBoolBlend {blendOption=Linear}
    alphaScaleBiasClamp {bClampResult=true, bInterpResult=true, clampMax=0.25}
    boneToModify {boneName=wing_ring_l_01}
    rotation=<rot -13.276 -5.55 4.957>
   }
   n28 = "Transform (Modify) Bone - Bone: wing_index_l_01"!(ComponentPose=@n27, Alpha=0, Rotation="4.520,1.609,-1.347") -> (Pose) {
    rotationMode=BMM_Additive
    alphaBoolBlend {blendOption=Linear}
    alphaScaleBiasClamp {bClampResult=true, bInterpResult=true, clampMax=0.5}
    boneToModify {boneName=wing_index_l_01}
    rotation=<rot 4.52 1.609 -1.347>
   }
   n29 = "Transform (Modify) Bone - Bone: wing_shoulder_r"(ComponentPose=@n28, Alpha=@n23, Rotation="-3.546,29.936,-4.646") {
    rotationMode=BMM_Additive
    alphaBoolBlend {blendOption=Linear}
    alphaScaleBiasClamp {bClampResult=true, bInterpResult=true, clampMax=0.65, interpSpeedDecreasing=2, interpSpeedIncreasing=15}
    boneToModify {boneName=wing_shoulder_r}
    rotation=<rot -3.546 29.936 -4.646>
   }
   n30 = "Transform (Modify) Bone - Bone: wing_wrist_r"(ComponentPose=@n29, Alpha=@n23, Rotation="20.023,4.529,15.773") {
    rotationMode=BMM_Additive
    alphaBoolBlend {blendOption=Linear}
    alphaScaleBiasClamp {bClampResult=true, bInterpResult=true, clampMax=0.65, interpSpeedDecreasing=2, interpSpeedIncreasing=15}
    boneToModify {boneName=wing_wrist_r}
    rotation=<rot 20.023 4.529 15.773>
   }
   n31 = "Transform (Modify) Bone - Bone: wing_pinky_r_01"!(ComponentPose=@n30, Alpha=@n23, Rotation="14.884,4.138,12.483") -> (Pose) {
    rotationMode=BMM_Additive
    alphaBoolBlend {blendOption=Linear}
    alphaScaleBiasClamp {bClampResult=true, bInterpResult=true, clampMax=0.5}
    boneToModify {boneName=wing_pinky_r_01}
    rotation=<rot 14.884 4.138 12.483>
   }
   n32 = "Transform (Modify) Bone - Bone: wing_ring_r_01"!(ComponentPose=@n31, Alpha=@n23, Rotation="13.276,5.55,4.957") -> (Pose) {
    rotationMode=BMM_Additive
    alphaBoolBlend {blendOption=Linear}
    alphaScaleBiasClamp {bClampResult=true, bInterpResult=true, clampMax=0.25}
    boneToModify {boneName=wing_ring_r_01}
    rotation=<rot 13.276 5.55 4.957>
   }
   n33 = "Transform (Modify) Bone - Bone: wing_index_r_01"!(ComponentPose=@n32, Alpha=0, Rotation="4.520,1.609,-1.347") -> (Pose) {
    rotationMode=BMM_Additive
    alphaBoolBlend {blendOption=Linear}
    alphaScaleBiasClamp {bClampResult=true, bInterpResult=true, clampMax=0.5}
    boneToModify {boneName=wing_index_r_01}
    rotation=<rot 4.52 1.609 -1.347>
   }
   Output Pose(Result=Component To Local(ComponentPose=@n33)) {layerGroup=None}
  ]}
 EventGraph {nodes [
   Event BlueprintUpdateAnimation()
   n1 = "To Float (Boolean)"(InBool=Get bIsSprinting())
   n2 = OR Boolean!(A=Get bIsTargeting(), B=Get bIsSkydiveDiveMode(), C:bool=Get bIsTacticalSprinting(), D:bool=Get bIsCrouching()) -> (ReturnValue)
   > Set DivePose_alpha(DivePose_alpha:double="Max (Float)"(A="Clamp (Float)"(Value=@n1, Min=0, Max=0.75), B="To Float (Boolean)"(InBool=@n2))) -> (Output_Get:double)
   > Set FlyPose_alpha(FlyPose_alpha:double="To Float (Boolean)"(InBool=Get bIsSkydiveFloating())) -> (Output_Get:double)
   n3 = Get bIsJumping()
   > Set JumpPose_alpha(JumpPose_alpha:double=SelectFloat(A=0, B="To Float (Boolean)"(InBool=@n3), bPickA=Get bIsSkyDiving())) -> (Output_Get:double)
   TryGetPawnOwner()
   TryGetPawnOwner()
  ]}
]
variables [
 deltaTimeX {type=number}
 DivePose_alpha {type=number}
 falling_anim_dyn_up_force_vec3 {type=Vector}
 FlyPose_alpha {type=number}
 JumpPose_alpha {type=number}
 Skydiving {type=FortAnimInput_Skydiving, default {skydiveAimPitchInterpSpeed=5}}
]
